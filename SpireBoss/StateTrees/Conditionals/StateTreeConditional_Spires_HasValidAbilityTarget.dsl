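; StateTree condition for the Spires boss that tests whether the large spire boss has a valid ability target.
; It gets the spire boss from its context and calls HasValidAbilityTarget on it.
; The bReturnInverse flag flips the result, so the same condition can also gate on having no valid target.
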(fn ReceiveTestCondition
  (bind _spireboss (Variables|Context|GetSpireBoss))
  (bind _hasvalidtarget (Class|BGALargeSpire|HasValidAbilityTarget _spireboss))
  (if (Variables|Default|GetbReturnInverse)
    (return (not _hasvalidtarget))
    (else
      (return _hasvalidtarget))))
