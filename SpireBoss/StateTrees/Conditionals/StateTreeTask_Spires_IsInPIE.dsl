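; A StateTree condition blueprint in the Spire Boss folder that checks whether the game is running in the editor or PIE, despite its Task name.
; It passes when the In Editor or PIE check equals its bValue variable, so it can test for either case.
; Its description override shows the expected value in the StateTree editor.
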(fn GetDescription
  (params (Formatting :type EStateTreeNodeFormatting%20Enum :default "RichText"))
  (return (Utilities|Text|FormatText "NSLOCTEXT(\"[E89AB072077CB37463A20B4909B58998]\", \"BB9F68BA445F497232B488A408C03D89\", \"Is in PIE is {Value}\")" (Variables|Default|GetbValue))))

(fn ReceiveTestCondition
  (return (== (Editor|InEditororPIE) (Variables|Default|GetbValue))))
